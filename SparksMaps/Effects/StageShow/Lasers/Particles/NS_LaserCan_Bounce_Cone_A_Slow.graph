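# Niagara stage-show laser can effect: a slowly rotating cone beam mesh paired with glow flare sprites.
# LaserBounce_Cone orients and spins the mesh, with User parameters driving color, brightness, cone width, rotation speed and length, which the Jam? flag shortens.
# Flare_Without_Occlusion bursts camera-offset sprites whose color comes from scratch inputs using a Particle Attribute Reader on the cone emitter.
# Dynamic material parameters pass color LUT overrides, plus distance-based LOD values for the beam.
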NiagaraSystem /SparksMaps/Effects/StageShow/Lasers/Particles/NS_LaserCan_Bounce_Cone_A_Slow
dependencies {
 usedDataInterfaces [&NiagaraDataInterfaceParticleRead]
 usedDynamicInputs [&UniformRangedVector, &Multiply_Vector2D_ByFloat]
 usedModules [&SystemState, /Niagara/Noise/Waveform]
}
emitters {
 Flare_Without_Occlusion {
  emitterUpdateScript [
   &EmitterState(Life Cycle Mode=::NewEnumerator1, Loop Behavior=::NewEnumerator1, Loop Duration Mode=::NewEnumerator1, UseLoopDelay=false)
   &SpawnBurst_Instantaneous(Spawn Count=3, Use Spawn Probability=false, UseLoopCountLimit=false)
  ]
  particleSpawnScript [
   &InitializeParticle(
    Color=&self:ScratchDynamicInput(New Particle Attribute Reader:NiagaraDataInterfaceParticleRead=DataInterface() {propertyValues {EmitterBinding {bindingMode=Other, emitterName=LaserBounce_Cone}}})
    Lifetime=12
    Position=&Add_Vector(A=/Niagara/DynamicInputs/Helpers/SimulationPosition(), B=<0 0 0>)
    Position Mode=::NewEnumerator1
    Sprite Rotation Angle=&RandomRangeFloat(Minimum=-1, Maximum=1)
    Sprite Rotation Mode=::NewEnumerator0
    Sprite Size Mode=::NewEnumerator3
    Uniform Sprite Size=150
    UsePositionOffset=false
    Write Material Random=true
   )
   /SetVariables(Particles.SpriteSize2=&Multiply_Vector2(A=$Particles.SpriteSize, B=&Vector2DFromFloat(Value=0.5))) {isSetParameters=true}
   /SetVariables(Particles.CameraOffsetLargeglow=40) {isSetParameters=true}
   &DynamicMaterialParameters(
    Index 2 Param 1=$User.Color LUT Custom U Position
    Index 2 Param 2=$User.UseMaterialColorOverride
    Module.Index2Param0Write=true
    Module.Index2Param1Write=true
    Module.Index2Param2Write=false
    Module.Index2Param3Write=false
    Write Parameter Index 0=false
    Write Parameter Index 2=true
   )
  ]
  particleUpdateScript [
   &ParticleState(
    Kill Particles When Lifetime Has Elapsed=false
    Let Infinitely Lived Particles Die When Emitter Deactivates=true
    Loop Particles Lifetime=true
   )
   &Color(Color=&self:ScratchDynamicInput_01(New Particle Attribute Reader:NiagaraDataInterfaceParticleRead=DataInterface() {propertyValues {EmitterBinding {bindingMode=Other, emitterName=LaserBounce_Cone}}}))
   &ScaleColor(Scale Alpha=0.5, ScaleA=true, ScaleRGB=true) {module=ScaleColor001}
   &CameraOffset(Camera Offset Amount=15)
  ]
  rendererClasses [&NiagaraSpriteRendererProperties]
  renderers [{rendererClass=&NiagaraSpriteRendererProperties}]
 }
 LaserBounce_Cone {
  emitterUpdateScript [
   &EmitterState(Life Cycle Mode=::NewEnumerator1, Loop Behavior=::NewEnumerator1, Loop Duration=1.5, UseLoopDelay=false)
   &SpawnBurst_Instantaneous(Use Spawn Probability=true, UseLoopCountLimit=false)
  ]
  particleSpawnScript [
   &InitializeParticle(
    Color=&LerpLinearColors(
     StartColor=&UniformRangedLinearColor(Minimum=<rgba 0 1 0.377 1>, Maximum=<rgba 10 0 0 1>)
     EndColor=$User.Color
     LerpFactor=$User.UseCustomColor
    )
    Lifetime=1.5
    Mesh Scale Mode=::NewEnumerator3
    Position=&Add_Vector(A=&SimulationPosition(), B=<0 0 0>)
    Position Mode=::NewEnumerator1
    Sprite Size=&Vector2DFromFloat(Value=768)
    Sprite Size Mode=::NewEnumerator4
    UsePositionOffset=false
    Write Material Random=true
   )
   &InitialMeshOrientation(Mesh Axis To Orient=<-1 0 0>, Mesh Orientation Mode=::NewEnumerator2, Orientation Vector=<0 0 1>, Use Rotation Vector=false) {module=InitialMeshOrientation001}
  ]
  particleUpdateScript [
   &ParticleState(Kill Particles When Lifetime Has Elapsed=false, Loop Particles Lifetime=true)
   &UpdateMeshOrientation(Rotation Rate=&Multiply_Float(A=0.01, B=$User.RotationSpeedMult))
   &ScaleMeshSize(Scale Factor=&Multiply_Vector(
     A=&MakeVector(X=0.65, Y=0.25, Z=0.25)
     B=&MakeVector(X=1, Y=$User.ConeScaleMult, Z=$User.ConeScaleMult)
    ))
   &ScaleMeshSize(Scale Factor=&VectorFromFloat(Value=&Multiply_Float(
      A=&MakeCustomFloatFromBool(Bool=Linked() {linkedVariable {name=User.Jam?, type=NiagaraBool, defaultValue=~, description=~}}, True Float=0.25, False Float=1.5)
      B=$User.LaserLengthMult
     ))) {module=ScaleMeshSize001}
   &ScaleColor(Scale RGB=&VectorFromFloat(Value=$User.BeamColorScale), ScaleA=true, ScaleRGB=true) {module=ScaleColor001}
   &SolveForcesAndVelocity(Clamp Velocity=false, Limit Acceleration=false)
   &DynamicMaterialParameters(
    Index 0 Param 1=$User.UseMaterialColorOverride
    Index 0 Param 2=&FN_LOD_Distance_Remap(Start Distance=250, End Distance=3000, Minimum Output=2, Maximum Output=4, Clamp Output=true)
    Index 0 Param 3=&FN_LOD_Distance_Remap(Start Distance=250, End Distance=3000, Minimum Output=6, Maximum Output=18, Clamp Output=true)
    Index 0 Param 4=$User.Color LUT Custom U Position
    Module.Param0WriteEnabled=true
    Module.Param1WriteEnabled=true
    Module.Param2WriteEnabled=true
    Module.Param3WriteEnabled=true
   )
  ]
  rendererClasses [&NiagaraMeshRendererProperties]
  renderers [{rendererClass=&NiagaraMeshRendererProperties, rendererIndex=0}]
 }
}
summary {
 systemName=NS_LaserCan_Bounce_Cone_A_Slow
 emitters [
  {emitterName=LaserBounce_Cone, rendererClasses [&NiagaraMeshRendererProperties]}
  {emitterName=Flare_Without_Occlusion, rendererClasses [&NiagaraSpriteRendererProperties]}
 ]
 userVariables [
  User.BeamColorScale {type=NiagaraFloat, defaultValue=0.05, description=""}
  User.Color {type=LinearColor, description="", defaultValue {struct=LinearColor, value=<rgba 0 4.5509 5 1>}}
  User.Color LUT Custom U Position {type=NiagaraFloat, defaultValue=0, description=""}
  User.ConeScaleMult {type=NiagaraFloat, defaultValue=2.5, description=""}
  User.Jam? {type=NiagaraBool}
  User.LaserLengthMult {type=NiagaraFloat, defaultValue=2.5, description=""}
  User.RotationSpeedMult {type=NiagaraFloat, defaultValue=1, description=""}
  User.UseCustomColor {type=NiagaraFloat, defaultValue=1, description=""}
  User.UseMaterialColorOverride {type=NiagaraFloat, defaultValue=1, description=""}
 ]
}
