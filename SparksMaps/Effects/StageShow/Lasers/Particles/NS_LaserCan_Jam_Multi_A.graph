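# Stage-show laser can effect that sequences three mesh laser looks, timed by system durations: random bouncing short-lived beams, a rotating cone, then a ring of waveform-animated beams.
# A sprite flare at the source reads its color from the bounce beams through a particle attribute reader.
# Beam colors cycle red, green and blue over particle life via curves.
# User parameters control length (shorter in Jam mode), wave motion and material color overrides.
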NiagaraSystem /SparksMaps/Effects/StageShow/Lasers/Particles/NS_LaserCan_Jam_Multi_A
dependencies {
 usedDataInterfaces [
  &NiagaraDataInterfaceCurve
  &NiagaraDataInterfaceColorCurve
  &NiagaraDataInterfaceParticleRead
  /Script/Niagara.NiagaraDataInterfaceOcclusion
 ]
 usedDynamicInputs []
 usedModules [
  &self:SystemSpawnScript.SystemScriptSource.SystemScriptGraph.NiagaraNodeAssignment_2.SetVariables_#1
  &self:SystemSpawnScript.SystemScriptSource.SystemScriptGraph.NiagaraNodeAssignment_1.SetVariables_#2
  &SystemState
  /Niagara/Noise/Waveform
  &self:SystemSpawnScript.SystemScriptSource.SystemScriptGraph.NiagaraNodeAssignment_0.SetVariables_#3
 ]
}
emitters {
 Flare_Without_Occlusion {
  emitterUpdateScript [
   &EmitterState(
    Inactive Response=::NewEnumerator1
    Life Cycle Mode=::NewEnumerator1
    Loop Behavior=::NewEnumerator1
    Loop Duration Mode=::NewEnumerator1
    UseLoopDelay=false
   )
   &SpawnBurst_Instantaneous(Use Spawn Probability=false, UseLoopCountLimit=false)
  ]
  particleSpawnScript [
   /Niagara/Modules/Spawn/Initialization/InitializeParticle(
    Color=&self:ScratchDynamicInput(New Particle Attribute Reader:NiagaraDataInterfaceParticleRead=DataInterface() {propertyValues {EmitterBinding {bindingMode=Other, emitterName=LaserBounce_BeamThin}}})
    Lifetime=12
    Mesh Scale=&UniformRangedVector(Minimum=<1 1 1>, Maximum=<2 2 2>)
    Module.Write Color=true
    Module.Write Lifetime=true
    Module.Write Mass=false
    Module.Write Material Random=true
    Module.Write Position=true
    Module.Write Scale=false
    Module.Write SpriteRotation=true
    Module.Write SpriteSize=true
    Position=&Add_Vector(A=/Niagara/DynamicInputs/Helpers/SimulationPosition(), B=<0 0 25>)
    Sprite Rotation=&RandomRangeFloat(Minimum=-1, Maximum=1)
    Sprite Size=&Multiply_Vector2D_ByFloat(Float=&RandomRangeFloat(Minimum=100, Maximum=300), Vector2D=<1 1>)
   ) {enabled=false, module=InitializeParticle001}
   /Niagara/Modules/Spawn/Initialization/V2/InitializeParticle(
    Color=&self:ScratchDynamicInput(New Particle Attribute Reader:NiagaraDataInterfaceParticleRead=DataInterface() {propertyValues {EmitterBinding {bindingMode=Other, emitterName=LaserBounce_BeamThin}}})
    Lifetime=12
    Position=&Add_Vector(A=&SimulationPosition(), B=<0 0 0>)
    Position Mode=::NewEnumerator1
    Sprite Rotation Angle=&RandomRangeFloat(Minimum=-1, Maximum=1)
    Sprite Rotation Mode=::NewEnumerator0
    Sprite Size Mode=::NewEnumerator3
    Uniform Sprite Size=150
    UsePositionOffset=false
    Write Material Random=true
   ) {enabled=false}
   &InitializeParticle(
    Color=$User.Color
    Lifetime=12
    Position=&Add_Vector(A=&SimulationPosition(), B=<0 0 0>)
    Position Mode=::NewEnumerator1
    Sprite Rotation Angle=&RandomRangeFloat(Minimum=-1, Maximum=1)
    Sprite Rotation Mode=::NewEnumerator0
    Sprite Size Mode=::NewEnumerator3
    Uniform Sprite Size=150
    UsePositionOffset=false
    Write Material Random=true
   ) {module=InitializeParticle002}
   /SetVariables(Particles.SpriteSize2=&Multiply_Vector2(A=$Particles.SpriteSize, B=&Vector2DFromFloat(Value=0.5))) {isSetParameters=true}
   /SetVariables(Particles.CameraOffsetLargeglow=40) {isSetParameters=true}
   &DynamicMaterialParameters(
    Index 2 Param 1=$User.Color LUT Custom U Position
    Index 2 Param 2=$User.UseMaterialColorOverride
    Module.Index2Param0Write=true
    Module.Index2Param1Write=true
    Module.Index2Param2Write=false
    Module.Index2Param3Write=false
    Write Parameter Index 0=false
    Write Parameter Index 2=true
   )
  ]
  particleUpdateScript [
   &ParticleState(
    Kill Particles When Lifetime Has Elapsed=false
    Let Infinitely Lived Particles Die When Emitter Deactivates=true
    Loop Particles Lifetime=true
   )
   &Color(Color=&self:ScratchDynamicInput_01(New Particle Attribute Reader:NiagaraDataInterfaceParticleRead=DataInterface() {propertyValues {EmitterBinding {bindingMode=Other, emitterName=LaserBounce_BeamThin}}}))
   &self:ScratchModule() {enabled=false, module=ScratchModule}
   &ScaleColor(Scale Alpha=&Lerp_Float(A=0.5, B=1, Alpha=$Particles.Visibility), ScaleA=true, ScaleRGB=true) {enabled=false}
   &ScaleSpriteSize(
    Scale Sprite Size Mode=::NewEnumerator0
    Uniform Scale Factor=&Lerp_Float(A=0.1, B=1, Alpha=$Particles.Visibility)
   ) {enabled=false}
   &ScaleColor(ScaleA=true, ScaleRGB=true) {module=ScaleColor001}
   &CameraOffset(Camera Offset Amount=15)
   &Color(Color=&ColorFromCurve(ColorCurve=DataInterface() {propertyValues {
       AlphaCurve {keys [<at 0 1>,<at 1 1>]}
       BlueCurve {keys [<at 0 0>,<at 0.333 0>,<at 0.667 1>,<at 1 0>]}
       GreenCurve {keys [<at 0 0>,<at 0.333 1>,<at 0.667 0>,<at 1 0>]}
       RedCurve {keys [<at 0 1>,<at 0.333 0>,<at 0.667 0>,<at 1 1>]}
      }})) {enabled=false, module=Color001}
  ]
  rendererClasses [&NiagaraSpriteRendererProperties]
  renderers [
   {rendererClass=&NiagaraSpriteRendererProperties}
   {rendererClass=&NiagaraSpriteRendererProperties, rendererIndex=1}
  ]
 }
 LaserBounce_BeamThin {
  emitterUpdateScript [
   &EmitterState(Inactive Response=::NewEnumerator1, Life Cycle Mode=::NewEnumerator1, Loop Delay=0.05, Loop Duration=$System.DurationA, UseLoopDelay=false)
   &SpawnRate(
    SpawnRate=&RandomRangeFloat(Minimum=6, Maximum=10, Evaluation Type=::NewEnumerator1)
    Use Spawn Probability=false
   )
   &SpawnRate(
    Spawn Probability=0.5
    SpawnRate=&RandomRangeFloat(Minimum=6, Maximum=10, Evaluation Type=::NewEnumerator1)
    Use Spawn Probability=true
   ) {module=SpawnRate001}
   &SpawnRate(
    SpawnRate=&RandomRangeFloat(Minimum=6, Maximum=10, Evaluation Type=::NewEnumerator1)
    Use Spawn Probability=false
   ) {enabled=false, module=SpawnRate002}
   &SpawnBurst_Instantaneous(
    Spawn Count=&RandomRangeInt(Minimum=1, Maximum=5)
    Spawn Probability=$System.EmitterSelect
    Spawn Time=$System.DurationA
    Use Spawn Probability=false
    UseLoopCountLimit=false
   ) {enabled=false}
  ]
  particleSpawnScript [
   /Niagara/Modules/Spawn/Initialization/InitializeParticle(
    Color=&LerpLinearColors(
     StartColor=&UniformRangedLinearColor(Minimum=<rgba 0 1 0.377 1>, Maximum=<rgba 10 0 0 1>)
     EndColor=$User.Color
     LerpFactor=$User.UseCustomColor
    )
    Lifetime=&RandomRangeFloat(Minimum=0.1, Maximum=0.15)
    Mesh Scale=&UniformRangedVector(Minimum=<1 1 1>, Maximum=<2 2 2>)
    Module.Write Color=true
    Module.Write Lifetime=true
    Module.Write Mass=false
    Module.Write Material Random=true
    Module.Write Position=true
    Module.Write Scale=true
    Module.Write SpriteRotation=false
    Module.Write SpriteSize=true
    Position=&Add_Vector(A=&SimulationPosition(), B=<0 0 0>)
    Sprite Size=&Vector2DFromFloat(Value=768)
   ) {enabled=false, module=InitializeParticle001}
   /Niagara/Modules/Spawn/Initialization/V2/InitializeParticle(
    Color=&LerpLinearColors(
     StartColor=&UniformRangedLinearColor(Minimum=<rgba 0 1 0.377 1>, Maximum=<rgba 10 0 0 1>)
     EndColor=$User.Color
     LerpFactor=$User.UseCustomColor
    )
    Lifetime Max=0.35
    Lifetime Min=0.25
    Lifetime Mode=::NewEnumerator1
    Mesh Scale Mode=::NewEnumerator3
    Position=&Add_Vector(A=&SimulationPosition(), B=<0 0 0>)
    Position Mode=::NewEnumerator1
    Sprite Size=&Vector2DFromFloat(Value=768)
    Sprite Size Mode=::NewEnumerator4
    UsePositionOffset=false
    Write Material Random=true
   )
   &InitializeParticle(
    Color=&LerpLinearColors(
     StartColor=&UniformRangedLinearColor(Minimum=<rgba 0 1 0.377 1>, Maximum=<rgba 10 0 0 1>)
     EndColor=$User.Color
     LerpFactor=$User.UseCustomColor
    )
    Lifetime=&RandomRangeFloat(Minimum=0.2, Maximum=0.25)
    Mesh Scale Mode=::NewEnumerator3
    Position=&Add_Vector(A=&SimulationPosition(), B=<0 0 0>)
    Position Mode=::NewEnumerator1
    Sprite Size=&Vector2DFromFloat(Value=768)
    Sprite Size Mode=::NewEnumerator4
    UsePositionOffset=false
    Write Material Random=true
   ) {enabled=false, module=InitializeParticle002}
   &InitialMeshOrientation(
    Mesh Axis To Orient=&RandomRangeVector(Minimum=<-1 -0.45 -0.45>, Maximum=<-1 0.45 0.45>)
    Mesh Orientation Mode=::NewEnumerator2
    Orientation Vector=<0 0 1>
    Use Rotation Vector=false
   ) {module=InitialMeshOrientation001}
  ]
  particleUpdateScript [
   &ParticleState()
   &KillParticles(Kill Particles=&SetBoolByFloatComparison(A=$System.EmitterSelect, B=0, Comparison Type=::NewEnumerator7))
   &ScaleMeshSize(Scale Factor=&MakeVector(
     X=&Multiply_Float(
      A=&MakeCustomFloatFromBool(Bool=Linked() {linkedVariable {name=User.Jam?, type=NiagaraBool, defaultValue=~, description=~}}, True Float=0.3, False Float=0.75)
      B=$User.LaserLengthMult
     )
     Y=1
     Z=1
    ))
   &ScaleColor(
    Scale Alpha=&FloatFromCurve(FloatCurve=DataInterface() {propertyValues {Curve {keys [<at -0.0046 1.0113 RCIM_Cubic>,<at 1.0011 0.002 RCIM_Cubic>]}}})
    ScaleA=true
    ScaleRGB=true
   ) {enabled=false}
   &ScaleColor(Scale RGB=&VectorFromFloat(Value=$User.BeamColorScale), ScaleA=true, ScaleRGB=true) {module=ScaleColor001}
   &SolveForcesAndVelocity(Clamp Velocity=false, Limit Acceleration=false)
   &DynamicMaterialParameters(
    Index 0 Param 1=$User.UseMaterialColorOverride
    Index 0 Param 2=&FN_LOD_Distance_Remap(Start Distance=350, End Distance=1000, Minimum Output=2, Maximum Output=4, Clamp Output=true)
    Index 0 Param 3=&FN_LOD_Distance_Remap(Start Distance=350, End Distance=1000, Minimum Output=5, Maximum Output=8, Clamp Output=true)
    Index 0 Param 4=$User.Color LUT Custom U Position
    Module.Param0WriteEnabled=true
    Module.Param1WriteEnabled=true
    Module.Param2WriteEnabled=true
    Module.Param3WriteEnabled=true
   )
   &ScaleColor(Scale Alpha=$System.EmitterSelect, ScaleA=true, ScaleRGB=true) {enabled=false, module=ScaleColor002}
   &ScaleColor(
    Scale Alpha=&FloatFromCurve(FloatCurve=DataInterface() {propertyValues {Curve {keys [
         {interpMode=RCIM_Cubic, tangentMode=RCTM_User, time=0, value=0}
         {interpMode=RCIM_Cubic, tangentMode=RCTM_User, time=0.5, value=1}
         {tangentMode=RCTM_User, time=1, value=0}
        ]}}})
    ScaleA=true
    ScaleRGB=false
   ) {enabled=false, module=ScaleColor003}
   &Color(Color=&ColorFromCurve(ColorCurve=DataInterface() {propertyValues {
       AlphaCurve {keys [<at 0 1>,<at 1 1>]}
       BlueCurve {keys [<at 0 0>,<at 0.333 0>,<at 0.667 1>,<at 1 0>]}
       GreenCurve {keys [<at 0 0>,<at 0.333 1>,<at 0.667 0>,<at 1 0>]}
       RedCurve {keys [<at 0 1>,<at 0.333 0>,<at 0.667 0>,<at 1 1>]}
      }}))
  ]
  rendererClasses [&NiagaraMeshRendererProperties]
  renderers [{rendererClass=&NiagaraMeshRendererProperties, rendererIndex=0}]
 }
 LaserRotate_Cone {
  emitterUpdateScript [
   &EmitterState()
   &SpawnBurst_Instantaneous(Spawn Time=$System.DurationA, Use Spawn Probability=false, UseLoopCountLimit=false)
  ]
  particleSpawnScript [
   /Niagara/Modules/Spawn/Initialization/InitializeParticle(
    Color=&LerpLinearColors(
     StartColor=&UniformRangedLinearColor(Minimum=<rgba 0 1 0.377 1>, Maximum=<rgba 10 0 0 1>)
     EndColor=$User.Color
     LerpFactor=$User.UseCustomColor
    )
    Lifetime=12
    Mesh Scale=&UniformRangedVector(Minimum=<1 1 1>, Maximum=<2 2 2>)
    Module.Write Color=true
    Module.Write Lifetime=true
    Module.Write Mass=false
    Module.Write Material Random=true
    Module.Write Position=true
    Module.Write Scale=true
    Module.Write SpriteRotation=false
    Module.Write SpriteSize=true
    Position=&Add_Vector(A=&SimulationPosition(), B=<0 0 10>)
    Sprite Size=&Vector2DFromFloat(Value=768)
   ) {enabled=false, module=InitializeParticle001}
   /Niagara/Modules/Spawn/Initialization/V2/InitializeParticle(
    Color=&LerpLinearColors(
     StartColor=&UniformRangedLinearColor(Minimum=<rgba 0 1 0.377 1>, Maximum=<rgba 10 0 0 1>)
     EndColor=$User.Color
     LerpFactor=$User.UseCustomColor
    )
    Lifetime=$System.DurationB
    Mesh Scale Mode=::NewEnumerator3
    Position=&Add_Vector(A=&SimulationPosition(), B=<0 0 10>)
    Position Mode=::NewEnumerator1
    Sprite Size=&Vector2DFromFloat(Value=768)
    Sprite Size Mode=::NewEnumerator4
    UsePositionOffset=false
    Write Material Random=true
   )
   &InitialMeshOrientation(
    Mesh Axis To Orient=&RandomRangeVector(Minimum=<-1 0 0>, Maximum=<-1 0 0>)
    Mesh Orientation Mode=::NewEnumerator2
    Orientation Vector=<0 0 1>
    Use Rotation Vector=false
   ) {module=InitialMeshOrientation001}
  ]
  particleUpdateScript [
   &KillParticles(Kill Particles=&SetBoolByFloatComparison(A=$System.NormalizedLoopAge, B=$System.EmitterRatio, Comparison Type=::NewEnumerator7)) {enabled=false}
   &ParticleState()
   &UpdateMeshOrientation(Rotation Rate=&Lerp_Float(A=-0.05, B=0.05, Alpha=$User.ReverseRotationDirection))
   &ScaleMeshSize(Scale Factor=&MakeVector(
     X=0.75
     Y=&FloatFromCurve(
      FloatCurve=DataInterface() {propertyValues {Curve {keys [<at 0 0.75>, <at 0.5 1>, {tangentMode=RCTM_User, time=1, value=0.75}]}}}
      Scale Curve=0.75
     )
     Z=&FloatFromCurve(
      FloatCurve=DataInterface() {propertyValues {Curve {keys [<at 0 0.75>, <at 0.5 1>, {tangentMode=RCTM_User, time=1, value=0.75}]}}}
      Scale Curve=0.75
     )
    )) {enabled=false}
   &ScaleMeshSize(Scale Factor=&VectorFromFloat(Value=&Multiply_Float(
      A=&MakeCustomFloatFromBool(Bool=Linked() {linkedVariable {name=User.Jam?, type=NiagaraBool, defaultValue=~, description=~}}, True Float=0.3, False Float=0.75)
      B=$User.LaserLengthMult
     ))) {module=ScaleMeshSize001}
   &ScaleColor(
    Scale Alpha=&FloatFromCurve(FloatCurve=DataInterface() {propertyValues {Curve {keys [<at -0.0046 1.0113 RCIM_Cubic>,<at 1.0011 0.002 RCIM_Cubic>]}}})
    ScaleA=true
    ScaleRGB=true
   ) {enabled=false}
   &ScaleColor(Scale RGB=&VectorFromFloat(Value=$User.BeamColorScale), ScaleA=true, ScaleRGB=true) {module=ScaleColor001}
   &SolveForcesAndVelocity(Clamp Velocity=false, Limit Acceleration=false)
   &DynamicMaterialParameters(
    Index 0 Param 1=$User.UseMaterialColorOverride
    Index 0 Param 2=&FN_LOD_Distance_Remap(Start Distance=350, End Distance=1000, Minimum Output=2, Maximum Output=4, Clamp Output=true)
    Index 0 Param 3=&FN_LOD_Distance_Remap(Start Distance=350, End Distance=1000, Minimum Output=5, Maximum Output=8, Clamp Output=true)
    Index 0 Param 4=$User.Color LUT Custom U Position
    Module.Param0WriteEnabled=true
    Module.Param1WriteEnabled=true
    Module.Param2WriteEnabled=true
    Module.Param3WriteEnabled=true
   )
   &ScaleColor(
    Scale Alpha=&FloatFromCurve(FloatCurve=DataInterface() {propertyValues {Curve {keys [
         {interpMode=RCIM_Cubic, tangentMode=RCTM_User, time=0, value=0}
         {interpMode=RCIM_Cubic, tangentMode=RCTM_User, time=0.5, value=1}
         {tangentMode=RCTM_User, time=1, value=0}
        ]}}})
    ScaleA=true
    ScaleRGB=false
   ) {enabled=false, module=ScaleColor003}
   &Color(Color=&ColorFromCurve(ColorCurve=DataInterface() {propertyValues {
       AlphaCurve {keys [<at 0 1>,<at 1 1>]}
       BlueCurve {keys [<at 0 0>,<at 0.333 0>,<at 0.667 1>,<at 1 0>]}
       GreenCurve {keys [<at 0 0>,<at 0.333 1>,<at 0.667 0>,<at 1 0>]}
       RedCurve {keys [<at 0 1>,<at 0.333 0>,<at 0.667 0>,<at 1 1>]}
      }}))
  ]
  rendererClasses [&NiagaraMeshRendererProperties]
  renderers [{rendererClass=&NiagaraMeshRendererProperties, rendererIndex=0}]
 }
 LaserWave_BeamThin {
  emitterUpdateScript [
   &EmitterState()
   &SpawnBurst_Instantaneous(
    Spawn Count=$User.BeamSpawnCount
    Spawn Time=&Add_Float(A=$System.DurationA, B=$System.DurationB)
    Use Spawn Probability=false
    UseLoopCountLimit=false
   )
  ]
  particleSpawnScript [
   /Niagara/Modules/Spawn/Initialization/InitializeParticle(
    Color=&LerpLinearColors(
     StartColor=&UniformRangedLinearColor(Minimum=<rgba 0 1 0.377 1>, Maximum=<rgba 10 0 0 1>)
     EndColor=$User.Color
     LerpFactor=$User.UseCustomColor
    )
    Lifetime=5
    Mesh Scale=&UniformRangedVector(Minimum=<1 1 1>, Maximum=<2 2 2>)
    Module.Write Color=true
    Module.Write Lifetime=true
    Module.Write Mass=false
    Module.Write Material Random=true
    Module.Write Position=true
    Module.Write Scale=true
    Module.Write SpriteRotation=false
    Module.Write SpriteSize=true
    Sprite Size=&Vector2DFromFloat(Value=768)
   ) {enabled=false, module=InitializeParticle001}
   /Niagara/Modules/Spawn/Initialization/V2/InitializeParticle(
    Color=&LerpLinearColors(
     StartColor=&UniformRangedLinearColor(Minimum=<rgba 0 1 0.377 1>, Maximum=<rgba 10 0 0 1>)
     EndColor=$User.Color
     LerpFactor=$User.UseCustomColor
    )
    Lifetime=$System.DurationC
    Mesh Scale Mode=::NewEnumerator3
    UsePositionOffset=false
    Write Material Random=true
   )
   &ShapeLocation(
    Apply To Particle Position=::NewEnumerator1
    MaskSpawnGroup=false
    Override Seed=false
    Ring / Disc Distribution Mode:ENiagaraSphereDistributionMode=::NewEnumerator1
    Ring / Disc Mode:ENiagaraRingDiscMode=::NewEnumerator1
    Ring Radius=1
    Rotation Angle=-30
    Rotation Angle Type:ENiagara_AngleInput=::NewEnumerator0
    Rotation Axis=<-1 0 0>
    Rotation Mode=::NewEnumerator1
    Shape Primitive=::NewEnumerator4
    U Position=&ReturnNormalizedExecIndex(Normalized Index Scale=0.05)
    UseOwnerScale=true
   )
   &InitialMeshOrientation(
    Mesh Axis To Orient=&MakeVector(X=-1, Y=0, Z=0)
    Mesh Orientation Mode=::NewEnumerator2
    Orientation Vector=$StackContext.ShapeLocation.ShapeVector
    Use Rotation Vector=false
   ) {module=InitialMeshOrientation001}
   &InitialMeshOrientation(Mesh Orientation Mode=::NewEnumerator2, Rotation=<-0.15 0 0>, Use Rotation Vector=true)
  ]
  particleUpdateScript [
   &ParticleState()
   &UpdateMeshOrientation(
    Facing Direction=&MakeVector(
     X=0.5
     Y=0
     Z=/Niagara/DynamicInputs/Helpers/Waveform(
      Global Input Position=&ReturnNormalizedExecIndex()
      Global Phase=&Multiply_Float(A=$Emitter.Age, B=$User.WaveSpeedMult)
      Global Amplitude Scale=&Lerp_Float(A=0, B=0.35, Alpha=$User.WaveIntensity)
      "Waveform [1]"=::NewEnumerator0
      "[1] Phase Scale"=$User.WavePhaseScale
      "[1] Frequency"=$User.WaveFrequency
      "[1] Amplitude Min/Max"=<-1 1>
     )
    )
    Facing Interpolation=false
    Orientation Method=::NewEnumerator1
    Override Quat=false
    TransformFacing=false
    TransformUp=false
   )
   &ScaleMeshSize(Scale Factor=&MakeVector(X=0.5, Y=1, Z=1)) {enabled=false}
   &ScaleMeshSize(Scale Factor=&VectorFromFloat(Value=&Multiply_Float(
      A=&MakeCustomFloatFromBool(Bool=Linked() {linkedVariable {name=User.Jam?, type=NiagaraBool, defaultValue=~, description=~}}, True Float=0.28, False Float=0.75)
      B=$User.LaserLengthMult
     ))) {module=ScaleMeshSize001}
   &ScaleColor(Scale RGB=&VectorFromFloat(Value=$User.BeamColorScale), ScaleA=true, ScaleRGB=true) {module=ScaleColor001}
   &ScaleColor(
    Scale Alpha=&FloatFromCurve(FloatCurve=DataInterface() {propertyValues {Curve {keys [<at -0.0046 1.0113 RCIM_Cubic>,<at 1.0011 0.002 RCIM_Cubic>]}}})
    ScaleA=true
    ScaleRGB=true
   ) {enabled=false}
   &SolveForcesAndVelocity(Clamp Velocity=false, Limit Acceleration=false)
   &DynamicMaterialParameters(
    Index 0 Param 1=$User.UseMaterialColorOverride
    Index 0 Param 2=&FN_LOD_Distance_Remap(Start Distance=350, End Distance=1000, Minimum Output=2, Maximum Output=4, Clamp Output=true)
    Index 0 Param 3=&FN_LOD_Distance_Remap(Start Distance=350, End Distance=1000, Minimum Output=5, Maximum Output=8, Clamp Output=true)
    Index 0 Param 4=$User.Color LUT Custom U Position
    Module.Param0WriteEnabled=true
    Module.Param1WriteEnabled=true
    Module.Param2WriteEnabled=true
    Module.Param3WriteEnabled=true
   )
   &ScaleColor(
    Scale Alpha=&FloatFromCurve(FloatCurve=DataInterface() {propertyValues {Curve {keys [
         {interpMode=RCIM_Cubic, tangentMode=RCTM_User, time=0, value=0}
         {interpMode=RCIM_Cubic, tangentMode=RCTM_User, time=0.5, value=1}
         {tangentMode=RCTM_User, time=1, value=0}
        ]}}})
    ScaleA=true
    ScaleRGB=false
   ) {enabled=false, module=ScaleColor003}
   &Color(Color=&ColorFromCurve(ColorCurve=DataInterface() {propertyValues {
       AlphaCurve {keys [<at 0 1>,<at 1 1>]}
       BlueCurve {keys [<at 0 0>,<at 0.333 0>,<at 0.667 1>,<at 1 0>]}
       GreenCurve {keys [<at 0 0>,<at 0.333 1>,<at 0.667 0>,<at 1 0>]}
       RedCurve {keys [<at 0 1>,<at 0.333 0>,<at 0.667 0>,<at 1 1>]}
      }}))
  ]
  rendererClasses [&NiagaraMeshRendererProperties]
  renderers [{rendererClass=&NiagaraMeshRendererProperties, rendererIndex=0}]
 }
}
summary {
 systemName=NS_LaserCan_Jam_Multi_A
 emitters [
  {emitterName=LaserBounce_BeamThin, rendererClasses [&NiagaraMeshRendererProperties]}
  {emitterName=Flare_Without_Occlusion, rendererClasses [&NiagaraSpriteRendererProperties]}
  {emitterName=LaserRotate_Cone, rendererClasses [&NiagaraMeshRendererProperties]}
  {emitterName=LaserWave_BeamThin, rendererClasses [&NiagaraMeshRendererProperties]}
 ]
 userVariables [
  User.BeamColorScale {type=NiagaraFloat, defaultValue=0.05, description=""}
  User.BeamSpawnCount {type=NiagaraInt32, defaultValue=35}
  User.Color {type=LinearColor, description="", defaultValue {struct=LinearColor, value=<rgba 0.4817 0.5 0.4734 1>}}
  User.Color LUT Custom U Position {type=NiagaraFloat, defaultValue=0, description=""}
  User.Jam? {type=NiagaraBool, defaultValue=true}
  User.LaserLengthMult {type=NiagaraFloat, defaultValue=1, description=""}
  User.ReverseRotationDirection {type=NiagaraFloat, defaultValue=0, description=""}
  User.UseCustomColor {type=NiagaraFloat, defaultValue=1, description=""}
  User.UseMaterialColorOverride {type=NiagaraFloat, defaultValue=0, description=""}
  User.WaveFrequency {type=NiagaraFloat, defaultValue=3, description=""}
  User.WaveIntensity {type=NiagaraFloat, defaultValue=1.5, description=""}
  User.WavePhaseScale {type=NiagaraFloat, defaultValue=3, description=""}
  User.WaveSpeedMult {type=NiagaraFloat, defaultValue=0.005, description=""}
 ]
}
